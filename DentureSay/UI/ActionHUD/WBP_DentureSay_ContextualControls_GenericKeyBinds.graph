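# Widget Blueprint for the DentureSay action HUD that shows generic key-bind feedback as contextual controls, derived from AthenaGenericKeyBindsFeedback.
# It configures the inherited EntryBox_Feedback as a vertical dynamic entry box, capped at a few entries, that builds its rows from the DentureSay ContextualControls LineContainer widget.
# A TutorialHighlight widget is also present; the event graph is empty and no variables are defined, so behavior comes from the parent class.
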WidgetBlueprint /DentureSay/UI/ActionHUD/WBP_DentureSay_ContextualControls_GenericKeyBinds
graphs [EventGraph {links [], nodes []}]
variables []
widgets {
 info {
  inheritedWidgetCount=2
  parentClass=/Script/FortniteUI.AthenaGenericKeyBindsFeedback
  rootWidgetClass=/Script/FortniteUI.FortDynamicEntryBox
  widgetCount=2
 }
 widgets [
  EntryBox_Feedback {
   widgetClassPath=&FortDynamicEntryBox
   bInherited=true
   widget=EntryBox_Feedback
   properties {
    entryBoxType=Vertical
    entryWidgetClass=&/WBP_DentureSay_ContextualControls_LineContainer.WBP_DentureSay_ContextualControls_LineContainer_C
    maximumEntries=7
    numDesignerPreviewEntries=1
   }
  }
  TutorialHighlight {widgetClassPath=&TutorialHighlight}
 ]
}
