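# DentureSay HUD widget, derived from AthenaHUDContextualControls, showing vehicle and generic keybind callouts in a tagged HUD element wrapper.
# PreConstruct sets or clears size-box overrides and restores full scale where small-text scaling is enabled; Construct hides vehicle keybind feedback.
# Extension events insert supplied widgets into the callout box or the initially collapsed vehicle spacer overlay.
# Exp-event start plays an alpha outro and finish plays the intro, both by name.
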WidgetBlueprint /DentureSay/UI/ActionHUD/WBP_DentureSay_ContextualControls
graphs [
 EventGraph {nodes [
   Event OnExpEventFinished() {is=title K2Node_Event}
   > PlayAnimationByName!(
    execute
    self:Fort Simple Widget Animations Panel Object Reference=Get FortSimpleWidgetAnimationsPanel() -> (FortSimpleWidgetAnimationsPanel:Fort Simple Widget Animations Panel Object Reference)
    Name:"Name (by ref)"=Get AnimIntroName() -> (AnimIntroName:Name=None)
    Speed:float=1
    StartAtTime:float=0
   ) -> (then, ReturnValue:bool)
   Event OnExpEventStarted() {is=title K2Node_Event}
   > PlayAnimationByName(self=Get FortSimpleWidgetAnimationsPanel(), Name=Get AnimOutroName() -> (AnimOutroName:Name=None), Speed=1, StartAtTime=0)
   Event PreConstruct()
   > SetSizeBoxOverride()
   > Branch(Condition=GetSmallTextScalingMultiplier!() -> (ScalingMultiplier:float, IsEnabledOnDevice:bool).IsEnabledOnDevice)
   > Set UserSpecifiedScale!(execute, UserSpecifiedScale:float=1, self:Scale Box Object Reference=Get ScaleBox_1() -> (ScaleBox_1:Scale Box Object Reference)) -> (then, Output_Get:float)
   n1 = Event OnControllerBindingCalloutExtensionAdded!() -> (OutputDelegate, then, ExtensionWidget:User Widget Object Reference) {is=title K2Node_Event}
   > n2 = AddChildToVerticalBox(self=Get ControllerCalloutBox() -> (ControllerCalloutBox:Vertical Box Object Reference), Content=@n1.ExtensionWidget)
   > SetHorizontalAlignment(self:Vertical Box Slot Object Reference=@n2, InHorizontalAlignment=HAlign_Fill)
   > SetVerticalAlignment(self:Vertical Box Slot Object Reference=@n2, InVerticalAlignment=VAlign_Fill)
   n3 = Event OnVehicleActionsSpacerExtensionAdded!() -> (OutputDelegate, then, ExtensionWidget:User Widget Object Reference) {is=title K2Node_Event}
   > n4 = AddChildToOverlay!(
    execute
    self:Overlay Object Reference=Get VehicleActionSpacerOverlay() -> (VehicleActionSpacerOverlay:Overlay Object Reference)
    Content:Widget Object Reference=@n3.ExtensionWidget
   ) -> (then, ReturnValue:Overlay Slot Object Reference)
   > SetHorizontalAlignment(self:Overlay Slot Object Reference=@n4, InHorizontalAlignment=HAlign_Fill)
   > SetVerticalAlignment(self:Overlay Slot Object Reference=@n4, InVerticalAlignment=VAlign_Fill)
   Event Construct()
   > HideFeedback(self:WBP BRGame UI Contextual Controls Vehicle Keybinds Object Reference=Get VehicleKeybinds() -> (VehicleKeybinds:WBP BRGame UI Contextual Controls Vehicle Keybinds Object Reference))
  ]}
 SetSizeBoxOverride {nodes [
   SetSizeBoxOverride() {is=title K2Node_FunctionEntry}
   > n1 = Branch(Condition=Get IsSizeBoxEnabled() -> (IsSizeBoxEnabled:bool))
   n2 = Get SizeBox() -> (SizeBox:Size Box Object Reference)
   > SetWidthOverride(self=@n2, InWidthOverride=Get SizeBoxWidthOverride() -> (SizeBoxWidthOverride:float))
   > SetHeightOverride(self=@n2, InHeightOverride=Get SizeBoxHeightOverride() -> (SizeBoxHeightOverride:float))
   ClearWidthOverride(execute=@n1.else, self:Size Box Object Reference=@n2)
   > ClearHeightOverride(self:Size Box Object Reference=@n2)
  ]}
]
variables [
 AnimIntroName {type=string, default=Intro}
 AnimOutroName {type=string, default=Outro}
 IsSizeBoxEnabled {type=boolean, default=true}
 SizeBoxHeightOverride {type=number, default=500}
 SizeBoxWidthOverride {type=number, default=500}
]
widgets {
 info {
  inheritedWidgetCount=0
  namedSlotCount=1
  parentClass=/Script/FortniteUI.AthenaHUDContextualControls
  rootWidgetClass=/Game/Athena/HUD/FortHUDElementWrapper.FortHUDElementWrapper_C
  widgetCount=9
 }
 widgets [
  FortHUDElementWrapper {
   widgetClassPath=/Game/Athena/HUD/FortHUDElementWrapper
   bIsVariable=true
   widget=FortHUDElementWrapper
   properties {
    elementTag {tagName=Athena.HUD.ContextualControls}
    hUDElementTag {gameplayTags [{tagName=Athena.HideForCapture}, {tagName=HUD.Athena.Hide.WithInventory}]}
   }
  }
  FortSimpleWidgetAnimationsPanel {
   widgetClassPath=/Script/FortniteUI.FortSimpleWidgetAnimationsPanel
   bIsVariable=true
   namedSlotHost=FortHUDElementWrapper
   widget=FortSimpleWidgetAnimationsPanel
   properties {
    animations [
     Intro {
      bIsEnabled=true
      onAnimationsFinished=()
      targets [{
        alpha=/Game/Athena/HUD/HUD_ContextualControls_IntroAlpha
        bResetOnFinish=false
        bShouldAnimateAlpha=true
        bShouldAnimateScale=false
        bShouldAnimateTranslation=false
        scale=null
        targetName=HorizontalBox_2
        translation=null
        materialScalarAnimations []
       }]
     }
     Outro {
      bIsEnabled=true
      onAnimationsFinished=()
      targets [{
        alpha=/Game/Athena/HUD/HUD_ContextualControls_OutroAlpha
        bResetOnFinish=false
        bShouldAnimateAlpha=true
        bShouldAnimateScale=false
        bShouldAnimateTranslation=false
        scale=null
        targetName=HorizontalBox_2
        translation=null
        materialScalarAnimations []
       }]
     }
    ]
    renderTransformPivot=<0 0.5>
   }
  }
  HorizontalBox_2 {
   widgetClassPath=&HorizontalBox
   parent=FortSimpleWidgetAnimationsPanel
   widget=HorizontalBox_2
   slot {class=FortSimpleWidgetAnimationsSlot}
  }
  VehicleActionSpacerOverlay {
   widgetClassPath=&Overlay
   bIsVariable=true
   parent=HorizontalBox_2
   widget=VehicleActionSpacerOverlay
   properties {visibility=Collapsed}
   slot {class=HorizontalBoxSlot, properties {horizontalAlignment=HAlign_Left}}
  }
  SizeBox {
   widgetClassPath=&SizeBox
   bIsVariable=true
   parent=HorizontalBox_2
   widget=SizeBox
   properties {bOverride_HeightOverride=true, bOverride_WidthOverride=true, heightOverride=500, widthOverride=500}
   slot {class=HorizontalBoxSlot, properties {horizontalAlignment=HAlign_Left}}
  }
  ScaleBox_1 {
   widgetClassPath=&ScaleBox
   bIsVariable=true
   parent=SizeBox
   widget=ScaleBox_1
   properties {stretch=UserSpecified, userSpecifiedScale=0.8}
   slot {class=SizeBoxSlot, properties {horizontalAlignment=HAlign_Left, verticalAlignment=VAlign_Bottom}}
  }
  ControllerCalloutBox {
   widgetClassPath=&VerticalBox
   bIsVariable=true
   parent=ScaleBox_1
   widget=ControllerCalloutBox
   properties {renderTransformPivot=<0 0.5>}
   slot {class=ScaleBoxSlot, properties {horizontalAlignment=HAlign_Left}}
  }
  VehicleKeybinds {
   widgetClassPath=/BattleRoyaleGameUI/HUD/ContextualControls/WBP_BRGameUI_ContextualControls_VehicleKeybinds
   bIsVariable=true
   parent=ControllerCalloutBox
   widget=VehicleKeybinds
   properties {maxLinesNeeded=11}
   slot {class=VerticalBoxSlot}
  }
  GenericKeyBinds {
   widgetClassPath=&/WBP_DentureSay_ContextualControls_GenericKeyBinds
   parent=ControllerCalloutBox
   widget=GenericKeyBinds
   properties {elementTag {tagName=HUD.Vehicle.Controls}, renderTransform {angle=0, scale=<1 1>, shear=<0 0>, translation=<0 0>}}
   slot {class=VerticalBoxSlot, properties {horizontalAlignment=HAlign_Left}}
  }
 ]
}
